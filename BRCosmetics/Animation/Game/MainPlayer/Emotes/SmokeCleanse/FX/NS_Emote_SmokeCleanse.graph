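# Niagara system for the SmokeCleanse emote, combining a wispy ribbon trail with a small flickering orange glow that follows the character's animation.
# FN_AnimTrail spawns short-lived anim-trail ribbons from bone or socket indices, distorts them with curl noise, and shapes color and width along the trail by normalized execution index.
# NE_Empty burst-spawns a looping-lifetime sprite pinned through a system Socket Reader, with waveform-driven alpha flicker and a camera offset.
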NiagaraSystem /BRCosmetics/Animation/Game/MainPlayer/Emotes/SmokeCleanse/FX/NS_Emote_SmokeCleanse
dependencies {
 usedDataInterfaces [&NiagaraDataInterfaceColorCurve, &NiagaraDataInterfaceCurve, &NiagaraDataInterfaceSocketReader]
 usedDynamicInputs []
 usedModules [
  &SystemState
  &self:SystemSpawnScript.SystemScriptSource.SystemScriptGraph.NiagaraNodeAssignment_0.SetVariables_#1
 ]
}
emitters {
 FN_AnimTrail {
  emitterSpawnScript [&FN_RibbonIDOffsetUtility()]
  emitterUpdateScript [
   &EmitterState()
   &FN_SpawnAnimTrail()
  ]
  particleSpawnScript [
   &FN_InitializeAnimTrailBoneIndecies()
   &FN_InitializeAnimTrailBoneIndecies(AnimTrail Selection=1, First Bone or Socket Index=2, Second Bone or Socket Index=3) {enabled=false, module=FN_InitializeAnimTrailBoneIndecies001}
   &FN_InitializeAnimTrail(AnimTrailWidth=3.5, Color=<rgba 0.2817 0.2741 0.2617 1>, Lifetime=0.3)
   &AddVelocity(Velocity=<0 400 0>) {enabled=false}
  ]
  particleUpdateScript [
   &CurlNoiseForce(Bias Noise=false, Noise Strength=&RandomRangeFloat(Minimum=75, Maximum=150), Pan Noise=false)
   &SolveForcesAndVelocity(Clamp Velocity=false, Limit Acceleration=false)
   &FN_UpdateAnimTrail() {module=FN_UpdateAnimTrail001}
   &ScaleColor(
    Curve Index=&ReturnNormalizedExecIndex()
    Linear Color Curve=DataInterface() {propertyValues {
      AlphaCurve {keys [<at 0.0648 0>,<at 0.7599 0.9>,<at 1 1>]}
      BlueCurve {keys [<at 0 1>,<at 0.7035 0.9117>,<at 1 0>]}
      GreenCurve {keys [<at 0 1>,<at 0.7035 0.9388>,<at 1 0.3073>]}
      RedCurve {keys [<at 0 1>,<at 0.7035 1>,<at 1 1>]}
     }}
    Scale Mode=::NewEnumerator2
   )
   &ScaleRibbonWidth(Ribbon Width Scale=&FloatFromCurve(
     FloatCurve=DataInterface() {propertyValues {Curve {keys [
         {tangentMode=RCTM_User, time=0, value=0}
         {interpMode=RCIM_Cubic, tangentMode=RCTM_User, time=0.2, value=0.65}
         {arriveTangent=0.3333, interpMode=RCIM_Cubic, leaveTangent=0.3333, time=0.7, value=0.275}
         {arriveTangent=2.4167, interpMode=RCIM_Cubic, leaveTangent=2.4167, time=0.8, value=0.85}
         {interpMode=RCIM_Cubic, tangentMode=RCTM_User, time=1, value=1}
        ]}}}
     CurveIndex=&ReturnNormalizedExecIndex()
     Scale Curve=1.05
    ))
  ]
  rendererClasses [&NiagaraRibbonRendererProperties]
  renderers [
   {rendererClass=&NiagaraRibbonRendererProperties}
   {rendererClass=&NiagaraRibbonRendererProperties, rendererIndex=1}
  ]
 }
 NE_Empty {
  emitterUpdateScript [
   &EmitterState(
    Inactive Response=::NewEnumerator1
    Life Cycle Mode=::NewEnumerator1
    Loop Behavior=::NewEnumerator1
    Loop Duration Mode=::NewEnumerator1
    UseLoopDelay=false
   )
   &SpawnBurst_Instantaneous(Use Spawn Probability=false, UseLoopCountLimit=false)
  ]
  particleSpawnScript [&InitializeParticle(
    Color=<rgba 1 0.1214 0.0004 0.35>
    Sprite Size Mode=::NewEnumerator3
    Uniform Sprite Size=18
    UsePositionOffset=false
    Write Material Random=true
   )]
  particleUpdateScript [
   &ParticleState(Kill Particles When Lifetime Has Elapsed=false, Loop Particles Lifetime=true)
   &SolveForcesAndVelocity(Clamp Velocity=false, Limit Acceleration=false)
   &SocketLocation(
    SamplePosOffset=true
    Sampled Position Offset=<0 0 -1>
    Socket Reader=$System.Socket Reader
    Socket Sampling Mode=::NewEnumerator2
    VelocitySpeedLimit=false
   )
   &ScaleColor(
    Scale Alpha=&Waveform("Waveform [1]"=::NewEnumerator9, "[1] Phase Scale"=2, "[1] Frequency"=7, "[1] Amplitude Min/Max"=<0 2>)
    ScaleA=true
    ScaleRGB=true
   )
   &CameraOffset(Camera Offset Amount=5)
  ]
  rendererClasses [&NiagaraSpriteRendererProperties]
  renderers [{rendererClass=&NiagaraSpriteRendererProperties}]
 }
}
summary {
 systemName=NS_Emote_SmokeCleanse
 emitters [
  {emitterName=FN_AnimTrail, rendererClasses [&NiagaraRibbonRendererProperties]}
  {emitterName=NE_Empty, rendererClasses [&NiagaraSpriteRendererProperties]}
 ]
 userVariables [User.RibbonIDOffset {type=NiagaraInt32}]
}
